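; Looping gameplay cue for the NanaimoBar charge effect, feeding the player's current weapon charge into its particles.
; On loop start it casts the target to the Athena player pawn, resets a count and nudges the first particle component along local X.
; A short-interval timer then copies charge percent into the "Charge Percent" particle parameter.
; The timer's looping flag is evaluated only at start (charge not yet full), and nothing here clears it.
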(event EventOnLoopingStart(EitherVFX-preferthisversion)
  (params (MyTarget :type Actor%20Object%20Reference) (Parameters :type Gameplay%20Cue%20Parameters%20Structure%20%28by%20ref%29) (ParticleComponents :type Array%20of%20FXSystem%20Component%20Object%20References) (AudioComponents :type Array%20of%20Audio%20Component%20Object%20References))
  (bind _asplayer_pawn_athena (Utilities|Casting|CastToPlayerPawn_Athena MyTarget))
  (Variables|Default|Setcount 0.0)
  (Transformation|AddLocalOffset (Utilities|Array|Get(acopy) ParticleComponents) "11.500000,0.000000,0.000000" false false)
  ; Is it Okay? worrying about perf
  (Utilities|Time|SetTimerbyEvent (delegate Custom|Charge%20Percent) 0.041 (not (> (FortWeapon|GetChargePercent (Weapon|GetCurrentWeapon _asplayer_pawn_athena)) 0.99999))))

(event Custom|Charge%20Percent
  (Effects|Components|ParticleSystem|SetFloatParameter (Utilities|Array|Get(acopy) (stale EventOnLoopingStart(EitherVFX-preferthisversion) ParticleComponents)) "Charge Percent" (FortWeapon|GetChargePercent (Weapon|GetCurrentWeapon (stale EventOnLoopingStart(EitherVFX-preferthisversion) _asplayer_pawn_athena)))))

(event EventBeginPlay)

(event Collision|EventActorBeginOverlap
  (params (OtherActor :type Actor%20Object%20Reference)))

(event EventTick
  (params (DeltaSeconds :type Float%20%28single-precision%29)))

(fn ConstructionScript)
