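; Gameplay cue notify for the Nanaimo Bar item projectile's hit, adding a radial shockwave effect on burst.
; It casts the target to B_Projectile_NanaimoBar and triggers the shockwave at the target's root component location.
; The projectile's charge percent scales the shockwave, blending one parameter between fixed bounds and another between two levels of the Size_additional scalable float with a multiplier.
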(fn OnBurst(EitherVFX-preferthisversion)
  (params (MyTarget :type Actor%20Object%20Reference) (Parameters :type Gameplay%20Cue%20Parameters%20Structure%20%28by%20ref%29) (ParticleComponents :type Array%20of%20FXSystem%20Component%20Object%20References) (AudioComponents :type Array%20of%20Audio%20Component%20Object%20References) (BurstCameraShakeInstance :type Camera%20Shake%20Base%20Object%20Reference) (BurstDecalInstance :type Decal%20Actor%20Object%20Reference))
  ;[1 0 0] VFX: Radial Force
  (bind _asb_projectile_nanaimo_bar (Utilities|Casting|CastToB_Projectile_NanaimoBar MyTarget))
  (Class|FortKismetLibrary|TriggerRadialShockwave (Variables|/Script/Engine.Actor|GetRootComponent (Variables|GetMyTarget)) (Transformation|GetWorldLocation (Variables|/Script/Engine.Actor|GetRootComponent (Variables|GetMyTarget))) (Math|Float|Lerp 1.3 1.7 (Variables|/Script/FortniteGame.FortProjectileBase|GetChargePercent _asb_projectile_nanaimo_bar)) (* (Math|Float|Lerp (ScalableFloat|GetValueAtLevel(double-precision) (Variables|Default|GetSize_additional)) (ScalableFloat|GetValueAtLevel(double-precision) (Variables|Default|GetSize_additional) 1.0) (Variables|/Script/FortniteGame.FortProjectileBase|GetChargePercent _asb_projectile_nanaimo_bar)) 1.25) 0.9)
  (return))
