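; Gameplay cue that configures the burst VFX played when the NanaimoBar water projectile hits.
; On burst, it orients a spawned particle component along the hit normal and reads charge percent from the target cast to the projectile.
; That charge, floored to half steps, indexes a Size scalable float to set FXRadius, while the raw charge drives FXScale.
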(fn OnBurst(EitherVFX-preferthisversion)
  (params (MyTarget :type Actor%20Object%20Reference) (Parameters :type Gameplay%20Cue%20Parameters%20Structure%20%28by%20ref%29) (ParticleComponents :type Array%20of%20FXSystem%20Component%20Object%20References) (AudioComponents :type Array%20of%20Audio%20Component%20Object%20References) (BurstCameraShakeInstance :type Camera%20Shake%20Base%20Object%20Reference) (BurstDecalInstance :type Decal%20Actor%20Object%20Reference))
  (bind (_returnvalue_bblockinghit _returnvalue_binitialoverlap _returnvalue_time _returnvalue_distance _returnvalue_location _returnvalue_impactpoint _returnvalue_normal) (Collision|BreakHitResult (Ability|GameplayCue|GetHitResult Parameters)))
  (bind _output (Utilities|Array|Get(acopy) ParticleComponents))
  (bind _asb_projectile_nanaimo_bar (Utilities|Casting|CastToB_Projectile_NanaimoBar MyTarget))
  (Transformation|SetWorldRotation _output (Math|Conversions|RotationFromXVector _returnvalue_normal) false false)
  (Effects|Components|ParticleSystem|SetFloatParameter _output "FXRadius" (ScalableFloat|GetValueAtLevel(double-precision) (Variables|Default|GetSize) (* (Math|Conversions|ToFloat(Integer) (Math|Float|Floor (* (Variables|/Script/FortniteGame.FortProjectileBase|GetChargePercent _asb_projectile_nanaimo_bar) 2.0))) 0.5)))
  (Effects|Components|ParticleSystem|SetFloatParameter _output "FXScale" (Variables|/Script/FortniteGame.FortProjectileBase|GetChargePercent _asb_projectile_nanaimo_bar))
  (return))
