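; Burst gameplay cue for the NanaimoBar charged muzzle effect, scaling its feedback by the cue's raw magnitude, which acts as the charge amount.
; OnBurst offsets the spawned particle component upward, lerps its FXScale between min and max charge scales, and scales the camera shake.
; Outside mobile and PC performance mode, at higher detail levels, it also spawns a magnitude-scaled FN_RadialForce physics field at the target's root.
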(fn OnBurst(EitherVFX-preferthisversion)
  (params (MyTarget :type Actor%20Object%20Reference) (Parameters :type Gameplay%20Cue%20Parameters%20Structure%20%28by%20ref%29) (ParticleComponents :type Array%20of%20FXSystem%20Component%20Object%20References) (AudioComponents :type Array%20of%20Audio%20Component%20Object%20References) (BurstCameraShakeInstance :type Camera%20Shake%20Base%20Object%20Reference) (BurstDecalInstance :type Decal%20Actor%20Object%20Reference))
  (bind (_normalizedmagnitude _rawmagnitude) (Ability|GameplayCue|BreakGameplayCueParameters Parameters))
  (bind _output (Utilities|Array|Get(acopy) ParticleComponents))
  (Transformation|AddLocalOffset _output "0.000000,0.000000,30.000000" false false)
  (Effects|Components|ParticleSystem|SetFloatParameter _output "FXScale" (Math|Float|Lerp (Variables|Default|GetScaleOnMinCharge) (Variables|Default|GetScaleOnMaxCharge) _rawmagnitude))
  ;[1 0 0] VFX: Radial Force
  (Variables|/Script/Engine.CameraShakeBase|SetShakeScale (* _rawmagnitude 0.5) BurstCameraShakeInstance)
  (if (Math|Boolean|ANDBoolean (not (Fortnite|IsMobilePlatform)) (not (Class|PlatformHelpersLibrary|IsPcPerfMode 0)) (> (Rendering|GetRenderingDetailMode) 1))
    (Game|SpawnActorfromClass 0 "/Game/Effects/Niagara/Blueprints/PhsyicsFields/FN_RadialForce.FN_RadialForce_C" (Math|Transform|MakeTransform (Transformation|GetWorldLocation (Variables|/Script/Engine.Actor|GetRootComponent MyTarget))) "Undefined" "MultiplyWithRoot" 0 3.0 0.7 (Math|Float|Lerp 6e+02 8e+02 _rawmagnitude) (Math|Float|Lerp 0.8 1.0 _rawmagnitude) 0.05 0)
    (return)))
